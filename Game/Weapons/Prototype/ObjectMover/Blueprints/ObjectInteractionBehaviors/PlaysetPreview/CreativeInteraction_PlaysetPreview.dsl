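; Creative-mode interaction behavior component for playset previews, used by the object move tool.
; When the interaction triggers, it calls the parent logic and, where cosmetic events can run, plays the move tool placement sound at the owner's location.
; Client-side interaction start calls the parent and checks for cosmetic execution, but has no effects wired to that check yet.
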(event EventBeginPlay
  (|CallParentFunction))

(event CreativeMoveTool|EventTriggerInteractionBehavior
  ;[1 1 1] Any effects associated with spawning the set should be done here.
  (|CallParentFunction)
  (Networking|CanExecuteCosmeticEvents
    (:True
      (Audio|PlaySoundatLocation "/Game/Sounds/Creative/Gadgets/MoveTool/MoveTool_Place_01_Cue.MoveTool_Place_01_Cue" (Transformation|GetActorLocation (Components|GetOwner))))))

(event CreativeMoveTool|EventStartCreativeInteractionOnClient
  (|CallParentFunction)
  (Networking|CanExecuteCosmeticEvents))
